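# Niagara system for the Aurum vehicle boost cosmetic, a gold-tinted sprite exhaust from a single continuously spawning emitter.
# Particles are partitioned across filtered skeletal mesh sockets, offset by a user vector and launched along negative X.
# Spawn groups vary lifetime, size and color scaling, while curves drive scale, fade and dynamic material parameters alongside SubUV animation.
# Particles inherit source movement with a falloff over life; the painted-variant override is disabled.
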NiagaraSystem /VehicleCosmetics/Boosts/Boost_Aurum/FX/NS_Boost_Aurum
dependencies {
 usedDataInterfaces [
  &NiagaraDataInterfaceCurve
  &NiagaraDataInterfaceSpriteRendererInfo
  &NiagaraDataInterfaceSkeletalMesh
 ]
 usedDynamicInputs []
 usedModules [
  &self:SystemSpawnScript.SystemScriptSource.SystemScriptGraph.NiagaraNodeAssignment_0.SetVariables_#1
  &SystemState
 ]
}
emitters {NewTextures {
  emitterUpdateScript [
   &EmitterState(Life Cycle Mode=::NewEnumerator1, Loop Duration=2, UseLoopDelay=false)
   &VC_FilteredSocketCount()
   &SpawnRate(SpawnRate=120, Use Spawn Probability=false)
  ]
  particleSpawnScript [
   /SetVariables(Particles.VisTag:NiagaraInt32=&SpawnGroup()) {isSetParameters=true}
   &VC_PaintedVariantsOverride() {enabled=false}
   &InitializeParticle(
    Color=<rgba 0.8522 0.359 0.0439 1>
    Lifetime=/VehicleCosmetics/FX/Niagara/Modules/MergeEmitters/VC_FloatBySpawnGroup(
     SpawnGroup or ExecIndex:NiagaraInt32=0
     CustomInt:NiagaraInt32=0
     MaximumSpawnGroups:NiagaraInt32=2
     Float SG 0:NiagaraFloat=&RandomRangeFloat(Minimum=0.42, Maximum=0.72)
     Float SG 1:NiagaraFloat=0.3
     Float SG 2:NiagaraFloat=0
     Float SG 3:NiagaraFloat=0
     Float SG 4:NiagaraFloat=0
     Float SG 5:NiagaraFloat=0
    )
    Sprite Rotation Mode=::NewEnumerator2
    Sprite Size Mode=::NewEnumerator3
    Uniform Sprite Size=&VC_FloatBySpawnGroup(
     SpawnGroup or ExecIndex:NiagaraInt32=0
     CustomInt:NiagaraInt32=0
     MaximumSpawnGroups:NiagaraInt32=2
     Float SG 0:NiagaraFloat=&RandomRangeFloat(Minimum=18, Maximum=22)
     Float SG 1:NiagaraFloat=20
     Float SG 2:NiagaraFloat=0
     Float SG 3:NiagaraFloat=0
     Float SG 4:NiagaraFloat=0
     Float SG 5:NiagaraFloat=0
    )
    UsePositionOffset=false
    Write Material Random=true
   )
   &PartitionParticles(Number Of Partitions=$Emitter.PortCount)
   &SkeletalMeshLocation(
    Bone / Socket Index=$Particles.PartitionParticles.Partition
    Bone / Socket Sampling Mode=::NewEnumerator5
    MaskSpawnGroup=false
    Mesh Sampling Type=::NewEnumerator6
    Sampled Position Offset=&AddVectorToPosition(Position=<-20 0 0>, Vector=$User.DTCarOffset)
    Skeletal Mesh=$System.SkeletalMesh
   )
   &AddVelocity(Velocity=&MakeCustomVectorFromBool(Bool=$User.bIsFrontendPreview, True Vector=<-1500 0 0>, False Vector=<-1500 0 0>))
  ]
  particleUpdateScript [
   &ParticleState()
   &ScaleSpriteSize(Uniform Curve Sprite Scale=DataInterface() {propertyValues {Curve {keys [
        {arriveTangent=20.6648, interpMode=RCIM_Cubic, leaveTangent=20.6648, tangentMode=RCTM_User, time=0, value=6}
        {arriveTangent=-0.4275, interpMode=RCIM_Cubic, leaveTangent=-0.4275, tangentMode=RCTM_User, time=0.5, value=9}
        <at 1 8 RCIM_Cubic>
       ]}}})
   &ScaleColor(
    Scale Alpha=&FloatFromCurve(FloatCurve=DataInterface() {propertyValues {Curve {
        defaultValue=1
        keys [
         {
          arriveTangent=5.5454
          arriveTangentWeight=0.2508
          interpMode=RCIM_Cubic
          leaveTangent=5.5454
          leaveTangentWeight=0.2508
          tangentMode=RCTM_User
          tangentWeightMode=RCTWM_WeightedBoth
          time=0
          value=0
         }
         {
          arriveTangent=2.6983
          arriveTangentWeight=0.1284
          interpMode=RCIM_Cubic
          leaveTangent=0.3131
          leaveTangentWeight=0.135
          tangentMode=RCTM_Break
          tangentWeightMode=RCTWM_WeightedBoth
          time=0.15
          value=0.7
         }
         {arriveTangent=-0.203, interpMode=RCIM_Cubic, leaveTangent=-0.203, tangentMode=RCTM_User, time=0.3, value=0.7}
         {arriveTangent=-2.7815, tangentMode=RCTM_User, time=0.8, value=-1}
        ]
       }}})
    Scale RGB=&MaskVectorBySpawnGroup(
     Masked Vector=&VectorFromFloat(Value=&FloatFromCurve(FloatCurve=DataInterface() {propertyValues {Curve {keys [
           <at 0 1>
           {arriveTangent=-16.0273, interpMode=RCIM_Cubic, leaveTangent=-16.0273, tangentMode=RCTM_User, time=0.1, value=8}
           {arriveTangent=-1.393, interpMode=RCIM_Cubic, leaveTangent=-1.393, tangentMode=RCTM_User, time=0.15, value=3}
           {arriveTangent=-1.8343, interpMode=RCIM_Cubic, leaveTangent=-1.8343, tangentMode=RCTM_User, time=0.7, value=1}
           {arriveTangent=-6.2407, tangentMode=RCTM_User, time=1, value=0}
          ]}}}))
     Passthrough Vector=&VectorFromFloat(Value=&FloatFromCurve(FloatCurve=DataInterface() {propertyValues {Curve {keys [<at 0 3>,<at 1 0>]}}}))
    )
    ScaleA=true
    ScaleRGB=true
   )
   &InheritSourceMovement(
    Applied Position Scale=&VectorFromFloat(Value=&FloatFromCurve(FloatCurve=DataInterface() {propertyValues {Curve {keys [
          {interpMode=RCIM_Cubic, tangentMode=RCTM_Break, time=0, value=1}
          {arriveTangent=-2.9025, interpMode=RCIM_Cubic, tangentMode=RCTM_Break, time=1, value=0}
         ]}}}))
    ApplyForce=false
    ApplyPosition=true
    ApplyPositionFalloff=false
   )
   &DynamicMaterialParameters(
    Index 0 Param 1=&FloatFromCurve(FloatCurve=DataInterface() {propertyValues {Curve {keys [
         {
          arriveTangent=-0.6388
          arriveTangentWeight=0.1743
          interpMode=RCIM_Cubic
          leaveTangent=-0.6388
          leaveTangentWeight=0.1743
          tangentMode=RCTM_User
          tangentWeightMode=RCTWM_WeightedBoth
          time=0
          value=10
         }
         {arriveTangent=-11.25, interpMode=RCIM_Cubic, leaveTangent=-11.25, time=0.3, value=9}
         {
          arriveTangent=-0.6785
          arriveTangentWeight=0.5588
          interpMode=RCIM_Cubic
          leaveTangent=-0.6785
          leaveTangentWeight=0.5588
          tangentMode=RCTM_User
          tangentWeightMode=RCTWM_WeightedBoth
          time=0.8
          value=1
         }
        ]}}})
    Index 0 Param 2=&MaskFloatBySpawnGroup(
     Masked Float=&FloatFromCurve(FloatCurve=DataInterface() {propertyValues {Curve {
         defaultValue=1
         keys [
          {arriveTangent=0.2001, interpMode=RCIM_Cubic, leaveTangent=0.2001, tangentMode=RCTM_User, time=0.2, value=0}
          {interpMode=RCIM_Cubic, tangentMode=RCTM_User, time=0.34, value=0.0112}
          {arriveTangent=-0.0433, interpMode=RCIM_Cubic, leaveTangent=-0.0433, time=0.5, value=0.0112}
          <at 0.6 0 RCIM_Cubic>
         ]
        }}})
     Passthrough Float=0
    )
    Index 0 Param 3=&FloatFromCurve(
     FloatCurve=DataInterface() {propertyValues {Curve {
        defaultValue=1
        keys [
         {
          arriveTangent=0.0849
          arriveTangentWeight=0.2709
          interpMode=RCIM_Cubic
          leaveTangent=0.0849
          leaveTangentWeight=0.2709
          tangentMode=RCTM_User
          tangentWeightMode=RCTWM_WeightedBoth
          time=0
          value=0
         }
         {tangentMode=RCTM_User, time=1, value=1}
        ]
       }}}
     Scale Curve=0
    )
    Index 0 Param 4=1
    Module.Index0Float0Write=true
    Module.Index0Float1Write=true
    Module.Index0Float2Write=false
    Module.Index0Float3Write=true
   )
   &SolveForcesAndVelocity(Clamp Velocity=false, Limit Acceleration=false)
   &SubUVAnimation(
    "Sprite Renderer "=DataInterface() {propertyValues {SpriteRenderer=&self:NewTextures_0.NiagaraSpriteRendererProperties_1}}
    UseEndFrame=false
    UseStartFrame=false
   )
  ]
  rendererClasses [&NiagaraSpriteRendererProperties]
  renderers [{rendererClass=&NiagaraSpriteRendererProperties}]
 }}
summary {
 systemName=NS_Boost_Aurum
 emitters [{emitterName=NewTextures, rendererClasses [&NiagaraSpriteRendererProperties]}]
 userVariables [
  User.bIsFrontendPreview {type=NiagaraBool}
  User.bIsPainted {type=NiagaraBool}
  User.ColorVariantID {type=NiagaraInt32}
  User.DTCarOffset {type=Vector3f, description="", defaultValue {struct=Vector3f, value=<0 0 0>}}
  User.intensidad {type=NiagaraFloat, defaultValue=1, description=""}
  User.PaintedColor {type=LinearColor, description="", defaultValue {struct=LinearColor, value=<rgba 1 1 1 1>}}
  User.PaintedScalar {type=NiagaraFloat, defaultValue=1, description=""}
  User.PortCount {type=NiagaraInt32, defaultValue=1}
 ]
}
